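# Post-process style AnimBP for the female PizzaParty JacketOff body, copying the attached parent's pose and refining it with a Control Rig.
# Weighted constraints anchor three zipper masters to body bones, and spring-driven AnimDynamics swing the zippers.
# Additive bone transforms adjust leg accessory deform bones by thigh-bend alphas computed each update from pelvis-relative thigh rotation.
# The backpack attach bone gets a fixed offset, plus another driven by the front-end Frontend_Idle_Bypass curve.
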AnimBlueprint /BRCosmetics/Characters/Player/Female/Medium/Bodies/F_MED_PizzaParty_JacketOff/Meshes/F_MED_PizzaParty_JacketOff_AnimBP
target_skeleton=&/F_MED_PizzaParty_JacketOff_Skeleton
graphs [
 AnimGraph {nodes [
   n1 = Control Rig(Source=Copy Pose From Mesh() {bUseAttachedParent=true}) {
    alphaBoolBlend {blendOption=Linear}
    controlRigAssetReference {blueprintRigClass=&/F_MED_PizzaParty_JacketOff_ControlRig.F_MED_PizzaParty_JacketOff_ControlRig_C}
   }
   n2 = "Transform (Modify) Bone - Bone: attach_backpack"(
    ComponentPose=Local To Component(LocalPose=@n1)
    Alpha=Get CustomIdle() -> (CustomIdle:double)
    Translation="0,-2.2,-1"
    Rotation="0,0,-8"
    Scale="1,1,1"
   ) {
    rotationMode=BMM_Additive
    rotationSpace=BCS_BoneSpace
    translationMode=BMM_Additive
    translationSpace=BCS_BoneSpace
    alphaBoolBlend {blendOption=Linear}
    boneToModify {boneName=attach_backpack}
    rotation {roll=-8}
    translation {y=-2.2, z=-1}
   }
   n3 = "Constraint - dyn_zipperA_mstr (PPPP)"!(
    ComponentPose=@n2
    Alpha=1
    ConstraintWeights_0=0.439
    ConstraintWeights_1:float=0.216
    ConstraintWeights_2:float=0.277
    ConstraintWeights_3:float=0.068
   ) -> (Pose) {
    alphaBoolBlend {blendOption=Linear}
    boneToModify {boneName=dyn_zipperA_mstr}
    constraintSetup [{targetBone {boneName=spine_05}}, {targetBone {boneName=neck_01}}, {targetBone {boneName=neck_02}}, {targetBone {boneName=head}}]
    constraintWeights [1,1,1,1]
   }
   n4 = "Constraint - dyn_zipperB_mstr (PPP)"!(ComponentPose=@n3, Alpha=1, ConstraintWeights_0=0.045, ConstraintWeights_1:float=0.432, ConstraintWeights_2:float=0.523) -> (Pose) {
    alphaBoolBlend {blendOption=Linear}
    boneToModify {boneName=dyn_zipperB_mstr}
    constraintSetup [{targetBone {boneName=pelvis}}, {targetBone {boneName=thigh_l}}, {targetBone {boneName=thigh_twist_01_l}}]
    constraintWeights [1,1,1]
   }
   n5 = "Constraint - dyn_zipperC_mstr (PPP)"!(ComponentPose=@n4, Alpha=1, ConstraintWeights_0=0.045, ConstraintWeights_1:float=0.432, ConstraintWeights_2:float=0.523) -> (Pose) {
    alphaBoolBlend {blendOption=Linear}
    boneToModify {boneName=dyn_zipperC_mstr}
    constraintSetup [{targetBone {boneName=pelvis}}, {targetBone {boneName=thigh_r}}, {targetBone {boneName=thigh_twist_01_r}}]
    constraintWeights [1,1,1]
   }
   n6 = Anim Dynamics - Bone: dyn_zipperA!(ComponentPose=@n5, Alpha=1, GravityOverride:Vector="0,0,-200", ExternalForce="0,0,0") -> (Pose) {
    angularDampingOverride=0.9
    angularSpringConstant=250
    bAngularSpring=true
    bOverrideAngularDamping=true
    bUseGravityOverride=true
    gravityScale=0
    alphaBoolBlend {blendOption=Linear}
    boundBone {boneName=dyn_zipperA}
    componentAppliedLinearAccClamp=<2000 2000 2000>
    componentLinearAccScale=<0.3 0.3 0.3>
    componentLinearVelScale=<0.8 0.8 0.8>
    physicsBodyDefinitions [{
      boundBone {boneName=dyn_zipperA}
      boxExtents=<5 10 10>
      constraintSetup {angularLimitsMax=<30 10 0>, angularLimitsMin=<-30 -10 -30>, angularTarget=<1 0 0>}
      localJointOffset=<0 2 3>
     }]
    relativeSpaceBone {boneName=root}
   }
   n7 = Anim Dynamics - Bone: dyn_zipperB!(ComponentPose=@n6, Alpha=1, GravityOverride:Vector="0,0,-200", ExternalForce="0,0,0") -> (Pose) {
    angularDampingOverride=0.999
    angularSpringConstant=100
    bAngularSpring=true
    bOverrideAngularDamping=true
    bUseGravityOverride=true
    bUsePlanarLimit=false
    gravityScale=0
    alphaBoolBlend {blendOption=Linear}
    boundBone {boneName=dyn_zipperB}
    componentAppliedLinearAccClamp=<900 900 900>
    componentLinearAccScale=<0.25 0.25 0.25>
    componentLinearVelScale=<0.8 0.8 0.8>
    physicsBodyDefinitions [{
      boundBone {boneName=dyn_zipperB}
      boxExtents=<5 5 5>
      constraintSetup {angularLimitsMax=<0 0 -10>, angularLimitsMin=<0 0 -40>, angularTarget=<1 0 0>}
      localJointOffset=<0 2 0>
     }]
    relativeSpaceBone {boneName=pelvis}
   }
   n8 = Anim Dynamics - Bone: dyn_zipperC!(ComponentPose=@n7, Alpha=1, GravityOverride:Vector="0,0,-100", ExternalForce="0,0,0") -> (Pose) {
    angularDampingOverride=0.99
    angularSpringConstant=80
    bAngularSpring=true
    bOverrideAngularDamping=true
    bUseGravityOverride=true
    gravityScale=0
    alphaBoolBlend {blendOption=Linear}
    boundBone {boneName=dyn_zipperC}
    componentAppliedLinearAccClamp=<900 900 900>
    componentLinearAccScale=<0.25 0.25 0.25>
    componentLinearVelScale=<0.8 0.8 0.8>
    physicsBodyDefinitions [{
      boundBone {boneName=dyn_zipperC}
      boxExtents=<5 6 6>
      constraintSetup {angularLimitsMax=<0 0 0>, angularLimitsMin=<0 0 -60>, angularTarget=<1 0 0>}
      localJointOffset=<2 2 2>
     }]
    relativeSpaceBone {boneName=root}
   }
   n9 = "Transform (Modify) Bone - Bone: deform_right_leg_acc"(
    ComponentPose=@n8
    Alpha=Get RightThighBendAlpha() -> (RightThighBendAlpha:double)
    Translation="6,-1,0"
    Rotation="0,-30,-10"
    Scale="0.8,0.8,1"
   ) {
    rotationMode=BMM_Additive
    rotationSpace=BCS_BoneSpace
    scaleMode=BMM_Additive
    scaleSpace=BCS_BoneSpace
    translationMode=BMM_Additive
    translationSpace=BCS_BoneSpace
    alphaBoolBlend {blendOption=Linear}
    boneToModify {boneName=deform_right_leg_acc}
    rotation {roll=-10, yaw=-30}
    scale=<0.8 0.8>
    translation=<6 -1>
   }
   n10 = "Transform (Modify) Bone - Bone: deform_left_leg_acc"(ComponentPose=@n9, Alpha=Get LeftThighBendAlpha() -> (LeftThighBendAlpha:double), Translation="0,3,0", Rotation="0,-15,10", Scale="1,0.5,1") {
    rotationMode=BMM_Additive
    rotationSpace=BCS_BoneSpace
    scaleMode=BMM_Additive
    scaleSpace=BCS_BoneSpace
    translationMode=BMM_Additive
    translationSpace=BCS_BoneSpace
    alphaBoolBlend {blendOption=Linear}
    boneToModify {boneName=deform_left_leg_acc}
    rotation {roll=10, yaw=-15}
    scale {y=0.5}
    translation {y=3}
   }
   n11 = "Transform (Modify) Bone - Bone: attach_backpack"!(ComponentPose=@n10, bAlphaBoolEnabled:bool=True, Translation="0,-2.5,0", Rotation="0,0,0", Scale="1,1,1") -> (Pose) {
    alphaInputType=Bool
    rotationMode=BMM_Additive
    rotationSpace=BCS_BoneSpace
    scaleMode=BMM_Additive
    scaleSpace=BCS_BoneSpace
    translationMode=BMM_Additive
    translationSpace=BCS_BoneSpace
    alphaBoolBlend {blendInTime=0.3, blendOption=CubicInOut, blendOutTime=0.3}
    boneToModify {boneName=attach_backpack}
    translation {y=-2.5}
   }
   Output Pose(Result=Component To Local(ComponentPose=@n11)) {layerGroup=None}
  ]}
 EventGraph {nodes [
   Event BlueprintUpdateAnimation()
   > n1 = Sequence()
   Set LeftThighBendAlpha(
    execute=@n1.then_1
    LeftThighBendAlpha:double=GetAlphaFromReferenceTransformDelta(SkeletalMeshComponent=Get Skel Mesh(), BoneName=thigh_l, BaseName=pelvis, InputRange=(X=-17,Y=-103), TwistAxis="1,0,0")
   ) -> (Output_Get:double)
   > Set RightThighBendAlpha(RightThighBendAlpha:double=GetAlphaFromReferenceTransformDelta(SkeletalMeshComponent=Get Skel Mesh(), BoneName=thigh_r, BaseName=pelvis, InputRange=(X=-17,Y=-103), TwistAxis="1,0,0")) -> (Output_Get:double)
   Branch(execute=@n1.then_0, Condition=Get bIsInFrontEnd())
   > Set CustomIdle(CustomIdle:double=GetCurveValue(self=GetOwnerFrontEndAnimInstance(), CurveName=Frontend_Idle_Bypass)) -> (Output_Get:double)
   Event OnCharacterCustomizationLoaded()
   > Set Skel Mesh(Skel Mesh=GetPartSkeletalMeshComponent())
  ]}
 OnNewFunction {nodes [
   OnNewFunction!() -> (then, Context:"Anim Update Context Structure (by ref)", Node:"Anim Node Reference Structure (by ref)") {is=title K2Node_FunctionEntry}
   > Return Node()
  ]}
 OnNewFunction_0 {nodes [
   OnNewFunction_0!() -> (then, Context:"Anim Update Context Structure (by ref)", Node:"Anim Node Reference Structure (by ref)") {is=title K2Node_FunctionEntry}
   > Return Node()
  ]}
]
variables [
 CustomIdle {type=number}
 LeftThighBendAlpha {type=number}
 RightThighBendAlpha {type=number}
 Skel Mesh {type=SkeletalMeshComponent}
]
